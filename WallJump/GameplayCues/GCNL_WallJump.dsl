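; GCNL_WallJump is a Fortnite gameplay cue for wall jumps that plays side-dependent Niagara effects and camera shake.
; On application it runs the parent logic, then passes a value from the cue parameters to the first particle component's LeftRight parameter.
; OnBurst feeds a cue parameter as direction to PlayCameraShake, which starts a camera-local shake on the target's player controller: left above zero, right down to -1 exclusive, otherwise none.
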(event EventBeginPlay)

(event Collision|EventActorBeginOverlap
  (params (OtherActor :type Actor%20Object%20Reference)))

(event EventTick
  (params (DeltaSeconds :type Float%20%28single-precision%29)))

(event EventOnApplication(Niagara)
  (params (MyTarget :type Actor%20Object%20Reference) (Parameters :type Gameplay%20Cue%20Parameters%20Structure%20%28by%20ref%29) (ParticleComponents :type Array%20of%20Niagara%20Particle%20System%20Component%20Object%20References) (AudioComponents :type Array%20of%20Audio%20Component%20Object%20References) (BurstCameraShakeInstance :type Camera%20Shake%20Base%20Object%20Reference) (BurstDecalInstance :type Decal%20Actor%20Object%20Reference))
  (|CallParentFunction MyTarget Parameters ParticleComponents AudioComponents BurstCameraShakeInstance BurstDecalInstance)
  (Utilities|FlowControl|Sequence
    (:then_0
      (Effects|Components|ParticleSystem|SetFloatParameter (Utilities|Array|Get(acopy) ParticleComponents) "LeftRight" (Ability|GameplayCue|BreakGameplayCueParameters Parameters)))))

(fn OnBurst
  (params (MyTarget :type Actor%20Object%20Reference) (Parameters :type Gameplay%20Cue%20Parameters%20Structure%20%28by%20ref%29))
  (|CallParentFunction MyTarget Parameters)
  (CallFunction|PlayCameraShake (Ability|GameplayCue|BreakGameplayCueParameters Parameters) MyTarget)
  (return true))

(fn PlayCameraShake
  (params (Direction :type Float%20%28single-precision%29) (Target :type Actor%20Object%20Reference))
  (bind _returnvalue (Fortnite|GetFortPlayerControllerfromActor Target))
  (bind _controller (LocalVariables|GetController))
  (bind _playercameramanager (Variables|/Script/Engine.PlayerController|GetPlayerCameraManager _controller))
  (Utilities|IsValid _returnvalue
    (:Is%20Valid
      (LocalVariables|SetController _returnvalue)
      (if (> Direction -1.0)
        (if (> Direction 0)
          (CameraShakes|StartCameraShake :self _playercameramanager :ShakeClass (Variables|Default|GetCameraShakeLeft) :Scale 1.0 :PlaySpace "CameraLocal")
          (else
            (CameraShakes|StartCameraShake :self _playercameramanager :ShakeClass (Variables|Default|GetCameraShakeRight) :Scale 1.0 :PlaySpace "CameraLocal")))))))

(fn ConstructionScript)
